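# Tutorial marker location sound cue on the Fort_UI sound class, with a main sound and a repeating ping.
# A mixer plays the Tutorial_MarkerLocation wave at full input volume alongside a half-volume branch that loops the ping wave behind a fixed three-second delay.
# Attenuation is overridden with HRTF spatialization and no stereo spread, so the cue is positioned in 3D despite its UI class.
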SoundCue /Game/Sounds/Fort_UI_Sounds/Tutorial/Tutorial_MarkerLocation_Cue
root=Mixer_0
nodes [
 {id=Delay_0, children [WavePlayer_6], properties {delayMax=3, delayMin=3}}
 {id=Looping_0, children [Delay_0]}
 {id=Mixer_0, children [WavePlayer_5,Looping_0], properties {inputVolume [1,0.5]}}
 {id=WavePlayer_5, properties {soundWaveAssetPtr=&/Tutorial_MarkerLocation}}
 {id=WavePlayer_6, properties {soundWaveAssetPtr=&/Tutorial_MarkerLocation_Ping}}
]
properties {
 bOverrideAttenuation=true
 maxDistance=4000
 soundClassObject=&Fort_UI
 volumeMultiplier=0.672
 attenuationOverrides {falloffDistance=4000, spatializationAlgorithm=SPATIALIZATION_HRTF, stereoSpread=0, attenuationShapeExtents {x=0}}
}
